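; Sound Library context for the MurkyVole auto turret that handles playback of weapon fire and other event sounds.
; Fire events either retrigger or spawn audio components on the owning actor's root, falling back to the actor's own; other events simply spawn attached.
; Spawned fire components are tracked, with a MetaSound output watcher updating an almost-finished flag.
; A cleanup helper removes queued component indices in reverse order.
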(fn CleanFireAudioComponents
  (bind _indextoclear (Variables|Default|GetIndexToClear))
  (if (> (Utilities|Array|Length _indextoclear) 0)
    (bind (_arrayindex _arrayelement) (Utilities|Array|ReverseforEachLoop _indextoclear)
      (:LoopBody
        (Utilities|Array|RemoveIndex (Variables|Default|GetFireAudioComponents) _arrayelement))
      (:Completed
        (Utilities|Array|Clear _indextoclear)
        (return false)))
    (else
      (return true))))

(fn GetRootComponent
  (params (Actor :type Actor%20Object%20Reference) (EventTag :type Gameplay%20Tag%20Structure))
  (bind _returnvalue (Actor|GetOwner Actor))
  (bind _rootcomponent (Variables|/Script/Engine.Actor|GetRootComponent Actor))
  (if (GameplayTags|MatchesTag EventTag "(TagName=\"SoundLibrary.Weapon.Fire\")")
    (Utilities|IsValid _returnvalue
      (:Is%20Valid
        (return (Variables|/Script/Engine.Actor|GetRootComponent _returnvalue)))
      (:Is%20Not%20Valid
        (return _rootcomponent)))
    (else
      (return _rootcomponent))))

(fn OnAlmostFinished
  (params (OutputName :type Name) (Output :type Meta%20Sound%20Output%20Structure%20%28by%20ref%29))
  (bind (_success _returnvalue) (MetaSoundOutput|GetBool Output))
  (if _success
    (Variables|Default|SetbAlmostFinished _returnvalue)
    (return)))

(fn Play
  (params (InEventData :type Sound%20Library%20Context%20Event%20Input%20Structure%20%28by%20ref%29))
  (bind (_sound _owner _eventtag) (Utilities|Struct|BreakSoundLibraryContextEventInput InEventData))
  (bind _rootcomponent (Helper|GetRootComponent _owner _eventtag))
  (bind _fireaudiocomponents (Variables|Default|GetFireAudioComponents))
  (if (GameplayTags|MatchesTag _eventtag "(TagName=\"SoundLibrary.Weapon.Fire\")")
    (bind _fireaudiocomponents_2 (Utilities|ShouldSpawnFireSound _sound)
      (:OnReTrigger
        (return _fireaudiocomponents_2 true))
      (:SpawnAudioComponent
        (bind _returnvalue (Audio|SpawnSoundAttached _sound _rootcomponent))
        (MetaSoundOutput|WatchOutput (WorldSubsystems|GetMetaSoundOutputSubsystem) _returnvalue "bOnAlmostFinished" (EventDispatchers|CreateEvent :!event OnAlmostFinished) "None" "None" :!class /Script/MetasoundEngine.MetaSoundOutputSubsystem)
        (Utilities|Array|AddUnique _fireaudiocomponents _returnvalue)
        (return _fireaudiocomponents true)))
    (else
      (bind _returnvalue_2 (Audio|SpawnSoundAttached _sound _rootcomponent))
      (return (Utilities|Array|MakeArray _returnvalue_2) true))))
